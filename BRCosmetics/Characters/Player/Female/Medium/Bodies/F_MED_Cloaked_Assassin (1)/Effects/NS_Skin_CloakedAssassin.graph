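# Cosmetic Niagara effect for the Cloaked Assassin skin, adding lightning, glows and runes to the character's skeletal mesh that intensify with User.KillCount.
# BigLightningEndPicker picks left or right bone/socket pairs; BigLightningRibbon and LightningSprites read its particles to draw arcing bolts, flashes and lights.
# GlowMesh and Runes spawn independently on the body, at kill-driven rates.
# Hand, head and foot hidden flags suppress effects through visibility tags and ribbon width.
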NiagaraSystem /BRCosmetics/Characters/Player/Female/Medium/Bodies/F_MED_Cloaked_Assassin/Effects/NS_Skin_CloakedAssassin
dependencies {
 usedDataInterfaces [
  &NiagaraDataInterfaceParticleRead
  /Script/Niagara.NiagaraDataInterfaceSimpleCounter
  &NiagaraDataInterfaceCurve
  &NiagaraDataInterfaceColorCurve
 ]
 usedDynamicInputs []
 usedModules [&SystemState]
}
emitters {
 BigLightningEndPicker {
  emitterUpdateScript [
   /SetVariables(Emitter.SetLoopDelay:NiagaraFloat=&UniformRangedFloat(Minimum=0.1, Maximum=0.9)) {isSetParameters=true}
   /SetVariables(Emitter.SetLoopDuration:NiagaraFloat=&Subtract_Float(A=1, B=$Emitter.SetLoopDelay)) {isSetParameters=true}
   &EmitterState(
    Life Cycle Mode=::NewEnumerator1
    Loop Delay=$Emitter.SetLoopDelay
    Loop Duration=$Emitter.SetLoopDuration
    Recalculate Duration Each Loop=true
    UseLoopDelay=true
   )
   &SpawnRate(
    SpawnRate=&ClampFloat(
     Float=&Add_Float(
      A=&Multiply_Float(A=&MakeFloatFromInt(Int=$User.KillCount), B=2.9)
      B=-5.7
     )
     Max=3
    )
    Use Spawn Probability=false
   )
  ]
  particleSpawnScript [
   /SetVariables(
    Particles.LeftOrRight:NiagaraInt32=&UniformRangedInt(Minimum=0, Maximum=1)
    Particles.LeftSource:NiagaraInt32=&UniformRangedInt(Minimum=0, Maximum=3)
    Particles.LeftTarget:NiagaraInt32=&UniformRangedInt(Minimum=4, Maximum=5)
    Particles.RightSource=9
    Particles.RightTarget:NiagaraInt32=&UniformRangedInt(Minimum=6, Maximum=8)
   ) {isSetParameters=true}
   /SetVariables(
    Particles.PickSourceBone:NiagaraFloat=&Lerp_Float(
     A=&MakeFloatFromInt(Int=$Particles.LeftSource)
     B=&MakeFloatFromInt(Int=$Particles.RightSource)
     Alpha=&MakeFloatFromInt(Int=$Particles.LeftOrRight)
    )
    Particles.PickTargetBone:NiagaraFloat=&Lerp_Float(
     A=&MakeFloatFromInt(Int=$Particles.LeftTarget)
     B=&MakeFloatFromInt(Int=$Particles.RightTarget)
     Alpha=&MakeFloatFromInt(Int=$Particles.LeftOrRight)
    )
   ) {isSetParameters=true}
   /Niagara/Modules/Spawn/Initialization/InitializeParticle(
    Color=&UniformRangedLinearColor(Minimum=<rgba 0.2 0 1 1>, Maximum=<rgba 2 0 3 1>, Color Channel Mode=::NewEnumerator2)
    Lifetime=&UniformRangedFloat(Minimum=0.2, Maximum=0.3)
    Module.Write Color=true
    Module.Write Lifetime=true
    Module.Write Mass=true
    Module.Write Material Random=true
    Module.Write Position=false
    Module.Write Scale=false
    Module.Write SpriteRotation=true
    Module.Write SpriteSize=true
    Position=&SimulationPosition()
    Sprite Rotation=90
    Sprite Size=&Vector2DFromFloat(Value=10)
   )
  ]
  particleUpdateScript [
   &SkeletalMeshLocation(
    Bone / Socket Index=&MakeIntFromFloat(Float=$Particles.PickSourceBone)
    Inherit Velocity=false
    MaskSpawnGroup=false
    Mesh Sampling Type=::NewEnumerator5
    Skeletal Mesh=$User.SkeletalMesh
    Socket Sampling Mode=::NewEnumerator5
   ) {module=SkeletalMeshLocation001}
   /SetVariables(Particles.HomePosition:Vector3f=$Output.SkeletalMeshLocation001.SampledPosition) {isSetParameters=true}
   &SkeletalMeshLocation(
    Apply To Particle Position=false
    Bone / Socket Index=&MakeIntFromFloat(Float=$Particles.PickTargetBone)
    Inherit Velocity=false
    MaskSpawnGroup=false
    Mesh Sampling Type=::NewEnumerator5
    Skeletal Mesh=$User.SkeletalMesh
    Socket Sampling Mode=::NewEnumerator5
   )
   /SetVariables(Particles.FriendPosition:Vector3f=$Output.SkeletalMeshLocation.SampledPosition) {isSetParameters=true}
   &SolveForcesAndVelocity(Clamp Velocity=false, Limit Acceleration=false)
   &ParticleState()
  ]
  rendererClasses [&NiagaraSpriteRendererProperties]
  renderers [
   {rendererClass=&NiagaraSpriteRendererProperties}
   {rendererClass=&NiagaraSpriteRendererProperties, rendererIndex=1}
  ]
 }
 BigLightningRibbon {
  emitterSpawnScript [/Game/Effects/Niagara/Modules/Attributes/FN_Source_v2(AttributeReader=DataInterface() {propertyValues {EmitterBinding {bindingMode=Other, emitterName=BigLightningEndPicker}}})]
  emitterUpdateScript [
   &EmitterState()
   /SetVariables(Emitter.ParticlesPerRibbon:NiagaraFloat=&UniformRangedFloat(Minimum=30, Maximum=75)) {isSetParameters=true}
   &SpawnRate(
    SpawnRate=/Game/Effects/Niagara/DynamicInputs/Attributes/FN_GetEmitterParticleCount(AttributeReader:NiagaraDataInterfaceParticleRead=$Emitter.AttributeReader, Multiplier:NiagaraFloat=$Emitter.ParticlesPerRibbon)
    Use Spawn Probability=false
   )
  ]
  particleSpawnScript [
   &self:GetInfoFromParent() {module=GetInfoFromParent}
   /SetVariables(Particles.AgeByParent:NiagaraFloat=&FloatFromCurve(FloatCurve=DataInterface() {propertyValues {Curve {keys [<at 0.2 1>,<at 0.8 0>]}}}, CurveIndex=$Particles.AgeImported)) {isSetParameters=true}
   /SetVariables(Particles.DistanceBetweenParents:NiagaraFloat=&ClampFloat(Float=&Add_Float(
      A=-1.55
      B=&Multiply_Float(
       A=0.05
       B=/Niagara/DynamicInputs/TypeConversions/VectorLengthSafe(Vector:Vector3f=&Subtract_Vector(A=$Particles.ParentPosition, B=$Particles.ParentFriendPosition))
      )
     ))) {isSetParameters=true}
   &InitializeRibbon(
    Color=<rgba 2.4952 0 6 1>
    Lifetime=0.3
    Module.Write Color=true
    Module.Write Lifetime=true
    Module.Write Mass=false
    Module.Write Material Random=true
    Module.Write Position=false
    Module.Write RibbonTwist=false
    Module.Write RibbonWidth=true
    Ribbon Width=&Multiply_Float(
     A=/Niagara/DynamicInputs/Math/LerpMultipleFloats(
      Float Count:ENiagara_MultipleLerpCount=::NewEnumerator0
      1:NiagaraFloat=2
      2:NiagaraFloat=&UniformRangedFloat(Minimum=25, Maximum=30)
      3:NiagaraFloat=2
      4:NiagaraFloat=?
      5:NiagaraFloat=?
      Alpha:NiagaraFloat=$Particles.AgeByParent
     )
     B=&Lerp_Float(A=0.75, B=1, Alpha=$Particles.DistanceBetweenParents)
    )
   )
   &InitializeParticle(
    Color=&ColorFromCurve(
     ColorCurve=DataInterface() {propertyValues {
       AlphaCurve {keys [<at 0 1>,<at 1 1>]}
       BlueCurve {keys [<at 0.0171 0>,<at 0.4894 0.0261>,<at 1.0012 1>]}
       GreenCurve {keys [<at 0.0171 0>,<at 0.4894 1>,<at 1.0012 0>]}
       RedCurve {keys [<at 0.0171 1>,<at 0.4894 0>,<at 1.0012 0>]}
      }}
     CurveIndex=$Particles.AgeByParent
    )
    Lifetime=0.7
    Module.Write Color=false
    Module.Write Lifetime=false
    Module.Write Mass=false
    Module.Write Material Random=true
    Module.Write Position=false
    Module.Write Scale=false
    Module.Write SpriteRotation=false
    Module.Write SpriteSize=true
    Sprite Size=&Vector2DFromFloat(Value=5)
   )
   /SetVariables(Particles.LeftOrRightMult:Vector3f=&Lerp_Vector(A=<1 1 1>, B=<-1 1 1>, Alpha=&MakeFloatFromInt(Int=$Particles.LeftOrRight))) {isSetParameters=true}
   /SetVariables(
    Particles.OffsetPerRibbon:Vector3f=&Multiply_Vector(
     A=&UniformRangedVector(Minimum=<10 -15 -10>, Maximum=<25 15 10>, Randomness Mode=::NewEnumerator1, Random Seed=$Particles.ParentIDInt, Fixed Random Seed=true)
     B=$Particles.LeftOrRightMult
    )
    Particles.OffsetRandom:Vector3f=&Multiply_VectorByFloat(
     Vector=&Multiply_Vector(A=&UniformRangedVector(Minimum=<-20 -20 -10>, Maximum=<5 20 10>), B=$Particles.LeftOrRightMult)
     Float=&Lerp_Float(A=0.2, B=1, Alpha=$Particles.DistanceBetweenParents)
    )
   ) {isSetParameters=true}
   /SetVariables(Particles.AddOffset:Vector3f=&Multiply_VectorByFloat(
     Vector=&Add_Vector(A=$Particles.OffsetPerRibbon, B=$Particles.OffsetRandom)
     Float=&FloatFromCurve(
      FloatCurve=DataInterface() {propertyValues {Curve {keys [
          {arriveTangent=2.9233, interpMode=RCIM_Cubic, leaveTangent=2.9233, tangentMode=RCTM_User, time=0.1, value=0}
          {interpMode=RCIM_Cubic, tangentMode=RCTM_User, time=0.4951, value=1}
          {arriveTangent=-2.7178, interpMode=RCIM_Cubic, leaveTangent=-2.7178, tangentMode=RCTM_User, time=0.9, value=0}
         ]}}}
      CurveIndex=$Particles.AgeByParent
     )
    )) {isSetParameters=true}
   &AddVelocity(Velocity=&Multiply_VectorByFloat(
     Vector=$Particles.OffsetRandom
     Float=&FloatFromCurve(
      FloatCurve=DataInterface() {propertyValues {Curve {keys [
          {arriveTangent=0.0416, interpMode=RCIM_Cubic, leaveTangent=0.0416, tangentMode=RCTM_User, time=0, value=0}
          {interpMode=RCIM_Cubic, tangentMode=RCTM_User, time=0.5, value=1}
          {arriveTangent=-0.0664, interpMode=RCIM_Cubic, leaveTangent=-0.0664, tangentMode=RCTM_User, time=1, value=0}
         ]}}}
      CurveIndex=$Particles.AgeByParent
      Scale Curve=&Multiply_Float(A=-10, B=$Particles.DistanceBetweenParents)
     )
    ))
  ]
  particleUpdateScript [
   &ParticleState()
   &SkeletalMeshLocation(
    Apply To Particle Position=false
    Bone / Socket Index=&MakeIntFromFloat(Float=$Particles.SourceBone)
    Inherit Velocity=false
    MaskSpawnGroup=false
    Mesh Sampling Type=::NewEnumerator5
    Skeletal Mesh=$User.SkeletalMesh
    Socket Sampling Mode=::NewEnumerator5
   ) {module=SkeletalMeshLocation001}
   /SetVariables(
    Particles.SourcePosition:Vector3f=$Output.SkeletalMeshLocation001.SampledPosition
    Particles.SourceRotation:Quat4f=$Output.SkeletalMeshLocation001.SampledRotation
   ) {isSetParameters=true}
   &SkeletalMeshLocation(
    Apply To Particle Position=false
    Bone / Socket Index=&MakeIntFromFloat(Float=$Particles.TargetBone)
    Inherit Velocity=false
    MaskSpawnGroup=false
    Mesh Sampling Type=::NewEnumerator5
    Skeletal Mesh=$User.SkeletalMesh
    Socket Sampling Mode=::NewEnumerator5
   )
   /SetVariables(
    Particles.TargetPosition:Vector3f=$Output.SkeletalMeshLocation.SampledPosition
    Particles.TargetRotation:Quat4f=$Output.SkeletalMeshLocation.SampledRotation
   ) {isSetParameters=true}
   /SetVariables(Particles.ParentsLerp:Vector3f=&Lerp_Vector(
     A=&Add_Vector(A=$Particles.SourcePosition, B=&Lerp_Vector(A=<0 0 -5>, B=<0 0 5>, Alpha=$Particles.NormalizedAge))
     B=$Particles.TargetPosition
     Alpha=&ClampFloat(Float=&Add_Float(A=-0.1, B=&Multiply_Float(A=1.2, B=$Particles.AgeByParent)))
    )) {isSetParameters=true}
   &Drag(Drag=5, Use Linear Drag=true, Use Rotational Drag=false)
   &SolveForcesAndVelocity(Clamp Velocity=false, Limit Acceleration=false)
   &ScaleRibbonWidth(Ribbon Width Scale=&FloatFromCurve(
     FloatCurve=DataInterface() {propertyValues {Curve {keys [
         {arriveTangent=7.8756, interpMode=RCIM_Cubic, leaveTangent=7.8756, tangentMode=RCTM_User, time=0, value=0}
         {interpMode=RCIM_Cubic, tangentMode=RCTM_User, time=0.2369, value=1}
         {arriveTangent=-2.4215, interpMode=RCIM_Cubic, leaveTangent=-2.4215, tangentMode=RCTM_User, time=1, value=0}
        ]}}}
     Scale Curve=&MakeCustomFloatFromBool(Bool=$User.bisHandHidden, True Float=0, False Float=1)
    ))
   &ScaleColor(
    Scale RGB=&VectorFromFloat(Value=&LerpMultipleFloats(
      Float Count:ENiagara_MultipleLerpCount=::NewEnumerator2
      1:NiagaraFloat=1.5
      2:NiagaraFloat=0.5
      3:NiagaraFloat=0.5
      4:NiagaraFloat=0.5
      5:NiagaraFloat=0.1
      Alpha:NiagaraFloat=$Particles.NormalizedAge
     ))
    ScaleA=false
    ScaleRGB=true
   )
   &self:TransformToBoneSpace(
    BonePosition:Vector3f=$Particles.ParentsLerp
    BoneRotation:Quat4f=$Particles.TargetRotation
    Position:Vector3f=&Multiply_VectorByFloat(Vector=$Particles.AddOffset, Float=&Lerp_Float(A=-0.5, B=0.5, Alpha=$Particles.NormalizedAge))
   ) {module=TransformToBoneSpace}
  ]
  rendererClasses [&NiagaraRibbonRendererProperties, &NiagaraSpriteRendererProperties]
  renderers [
   {rendererClass=&NiagaraRibbonRendererProperties}
   {rendererClass=&NiagaraSpriteRendererProperties, rendererIndex=1}
  ]
 }
 GlowMesh {
  emitterUpdateScript [
   &EmitterState()
   &SpawnRate(
    SpawnRate=&Multiply_Float(A=&ClampFloat(Float=&MakeFloatFromInt(Int=$User.KillCount)), B=24)
    Use Spawn Probability=false
   )
  ]
  particleSpawnScript [
   &InitializeParticle(
    Color=&UniformRangedLinearColor(Minimum=<rgba 0.5185 0 0.707 0.005>, Maximum=<rgba 1 0 0.7076 0.015>)
    Lifetime=2
    Mesh Scale=&VectorFromFloat(Value=&UniformRangedFloat(Minimum=15, Maximum=30))
    Module.Write Color=true
    Module.Write Lifetime=true
    Module.Write Mass=true
    Module.Write Material Random=true
    Module.Write Position=true
    Module.Write Scale=true
    Module.Write SpriteRotation=true
    Module.Write SpriteSize=false
    Position=&SimulationPosition()
    Sprite Rotation=&UniformRangedFloat(Minimum=0, Maximum=360)
   )
   /SetVariables(Particles.ScaleGnomon:Vector3f=<1 1 1>) {isSetParameters=true}
   /SetVariables(
    Particles.CameraOffset=-2
    Particles.HideLight:NiagaraFloat=&ClampFloat(Float=&Add_Float(
      A=&MakeFloatFromInt(Int=&ModuloInt(
        A=/Niagara/DynamicInputs/DataInterface/GenerateIntFromCounter(SimpleCounter:NiagaraDataInterfaceSimpleCounter=DataInterface() {propertyValues {GpuSyncMode=None, InitialValue=0}})
        B=24
       ))
      B=-22
     ))
    Particles.MeshOrientation=&MultiplyQuaternion(
     Quaternion A=&OrientMeshToVector(Look At Direction=<0 0 -1>, Reference Axis Vector=<1 0 0>)
     Quaternion B=&MakeQuatFromAxisAngle(Axis=<0 0 1>, AngleInDegrees=&UniformRangedFloat(Minimum=-360, Maximum=360))
    )
    Particles.ParticleRandom:NiagaraFloat=&UniformRangedFloat(Minimum=0, Maximum=1)
    Particles.PickBone:NiagaraInt32=&ModuloInt(
     A=&GenerateIntFromCounter(SimpleCounter:NiagaraDataInterfaceSimpleCounter=DataInterface() {propertyValues {GpuSyncMode=None, InitialValue=0}})
     B=4
    )
   ) {isSetParameters=true}
   /SetVariables(Particles.LightExponent=2, Particles.LightRadius=&Multiply_Float(A=10, B=$Particles.HideLight)) {isSetParameters=true}
  ]
  particleUpdateScript [
   /SetVariables(Particles.KillCountRamp:NiagaraFloat=&Multiply_Float(A=&MakeFloatFromInt(Int=$User.KillCount), B=0.3333)) {isSetParameters=true}
   &ParticleState()
   &DynamicMaterialParameters(
    Index 0 Param 1=&Multiply_Float(A=0.05, B=$Engine.Time)
    Index 0 Param 2=$Particles.ParticleRandom
    Index 0 Param 3=1
    Index 0 Param 4=&Lerp_Float(A=0.25, B=1, Alpha=$Particles.ParticleRandom)
    Module.Param0WriteEnabled=true
    Module.Param1WriteEnabled=true
    Module.Param2WriteEnabled=true
    Module.Param3WriteEnabled=true
   )
   &SkeletalMeshLocation(
    Bone / Socket Index=$Particles.PickBone
    Bone Sampling Mode=::NewEnumerator4
    Inherit Velocity=false
    MaskSpawnGroup=false
    Sampled Position Offset=&Lerp_Vector(
     A=<10 0 0>
     B=<0 8 0>
     Alpha=&ClampFloat(Float=&Add_Float(A=&MakeFloatFromInt(Int=$Particles.PickBone), B=-2))
    )
    Sampled Position Offset Coordinate Space:ENiagaraMeshTransforms=::NewEnumerator3
    Skeletal Mesh=$User.SkeletalMesh
   ) {module=SkeletalMeshLocation001}
   &ScaleColor(
    Scale Alpha=&FloatFromCurve(
     FloatCurve=DataInterface() {propertyValues {Curve {keys [
         {arriveTangent=0.0226, interpMode=RCIM_Cubic, leaveTangent=0.0226, tangentMode=RCTM_User, time=0, value=0}
         <at 0.25 1>
         {arriveTangent=0.0279, interpMode=RCIM_Cubic, leaveTangent=0.0279, tangentMode=RCTM_User, time=0.75, value=1}
         {arriveTangent=0.0181, interpMode=RCIM_Cubic, leaveTangent=0.0181, tangentMode=RCTM_User, time=1.0011, value=0.002}
        ]}}}
     Scale Curve=&Lerp_Float(A=0.2, B=1, Alpha=$Particles.KillCountRamp)
    )
    Scale RGB=&Lerp_Vector(A=<1 1 1>, B=&MakeVectorFromLinearColorRGB(Color=<rgba 0 0.2833 1 1>), Alpha=$Particles.NormalizedAge)
    ScaleA=true
    ScaleRGB=true
   )
   &ScaleMeshSize(Scale Factor=&VectorFromFloat(Value=&FloatFromCurve(FloatCurve=DataInterface() {propertyValues {Curve {keys [
          {arriveTangent=0.7209, interpMode=RCIM_Cubic, leaveTangent=0.7209, tangentMode=RCTM_User, time=-0.0046, value=0.7145}
          {arriveTangent=0.1347, interpMode=RCIM_Cubic, leaveTangent=0.1347, tangentMode=RCTM_User, time=0.4, value=1}
          {arriveTangent=-0.2593, interpMode=RCIM_Cubic, leaveTangent=-0.2593, tangentMode=RCTM_User, time=1.0011, value=0.85}
         ]}}})))
   /SetVariables(
    Particles.LightColor=&MultiplyLinearColor(Linear Color=$Particles.Initial.Color, Scale Factor=&MakeVector4_FromFloat(Float=3))
    Particles.LightRadius=&Multiply_Float(
     A=$Particles.Initial.LightRadius
     B=&FloatFromCurve(
      FloatCurve=DataInterface() {propertyValues {Curve {keys [
          {arriveTangent=0.0226, interpMode=RCIM_Cubic, leaveTangent=0.0226, tangentMode=RCTM_User, time=0, value=0}
          {arriveTangent=0.484, interpMode=RCIM_Cubic, leaveTangent=0.484, tangentMode=RCTM_User, time=0.3, value=1}
          {arriveTangent=-0.5029, interpMode=RCIM_Cubic, leaveTangent=-0.5029, tangentMode=RCTM_User, time=0.7, value=1}
          {arriveTangent=0.0181, interpMode=RCIM_Cubic, leaveTangent=0.0181, tangentMode=RCTM_User, time=1.0011, value=0.002}
         ]}}}
      Scale Curve=&Lerp_Float(A=0.2, B=1, Alpha=$Particles.KillCountRamp)
     )
    )
   ) {isSetParameters=true}
   /SetVariables(Particles.VisibilityTag=&self:Part_Hidden_PIck_Bone_Check(
     is Hand Hidden:NiagaraBool=$User.bisHandHidden
     is Head Hidden:NiagaraBool=$User.bisHeadHidden
     is Foot Hidden:NiagaraBool=$User.bisFootHidden
     Pick Bone:NiagaraInt32=$Particles.PickBone
    )) {isSetParameters=true, module=SetVariables_#1}
  ]
  rendererClasses [&NiagaraMeshRendererProperties, &NiagaraLightRendererProperties]
  renderers [
   {rendererClass=&NiagaraMeshRendererProperties, rendererIndex=0}
   {rendererClass=&NiagaraMeshRendererProperties, rendererIndex=1}
   {rendererClass=&NiagaraLightRendererProperties, rendererIndex=2}
  ]
 }
 LightningSprites {
  emitterSpawnScript [/SetVariables(Emitter.AttributeReader=DataInterface() {propertyValues {EmitterBinding {bindingMode=Other, emitterName=BigLightningEndPicker}}}) {isSetParameters=true}]
  emitterUpdateScript [
   &EmitterState(
    Life Cycle Mode=::NewEnumerator1
    Loop Delay=$Emitter.SetLoopDelay
    Loop Duration=$Emitter.SetLoopDuration
    Recalculate Duration Each Loop=true
    UseLoopDelay=true
   )
   &SpawnRate(
    SpawnRate=&FN_GetEmitterParticleCount(
     AttributeReader:NiagaraDataInterfaceParticleRead=$Emitter.AttributeReader
     Multiplier:NiagaraFloat=&UniformRangedFloat(Minimum=25, Maximum=50)
    )
    Use Spawn Probability=false
   )
  ]
  particleSpawnScript [
   &self:GetInfoFromParent() {module=GetInfoFromParent}
   &InitializeParticle(
    Color=<rgba 0.8317 0 2 1>
    Lifetime=&UniformRangedFloat(Minimum=0.2, Maximum=0.75)
    Module.Write Color=true
    Module.Write Lifetime=true
    Module.Write Mass=true
    Module.Write Material Random=true
    Module.Write Position=false
    Module.Write Scale=false
    Module.Write SpriteRotation=true
    Module.Write SpriteSize=true
    Position=&SimulationPosition()
    Sprite Rotation=&UniformRangedFloat(Minimum=0, Maximum=360)
    Sprite Size=&Vector2DFromFloat(Value=&UniformRangedFloat(Minimum=10, Maximum=20))
   )
   &DynamicMaterialParameters(
    Index 0 Param 1=&UniformRangedFloat(Minimum=0, Maximum=1)
    Index 0 Param 2=&UniformRangedFloat(Minimum=0.1, Maximum=1)
    Index 0 Param 3=1
    Module.Param0WriteEnabled=true
    Module.Param1WriteEnabled=true
    Module.Param2WriteEnabled=true
    Module.Param3WriteEnabled=false
   ) {module=DynamicMaterialParameters001}
   /SetVariables(
    Particles.FacingAdjustment:Vector3f=&UniformRangedVector(Minimum=<-1 -1 -1>, Maximum=<1 1 1>)
    Particles.SpriteFacing=&UniformRangedVector(Minimum=<-5 -5 -5>, Maximum=<5 5 5>)
    Particles.SpriteFacingAlt:Vector3f=&UniformRangedVector(Minimum=<-5 -5 -5>, Maximum=<5 5 5>)
    Particles.SubImageIndex=&UniformRangedFloat(Minimum=0, Maximum=3)
    Particles.SubImageIndexAlt:NiagaraFloat=&UniformRangedFloat(Minimum=0, Maximum=3)
   ) {isSetParameters=true}
   /SetVariables(
    Particles.LightBrightness=5
    Particles.LightExponent=1
    Particles.LightRadius=&Multiply_Float(
     A=&UniformRangedFloat(Minimum=8, Maximum=12)
     B=&ClampFloat(Float=&MakeFloatFromInt(Int=&Add_Integer(
        A=&ModuloInt(
         A=&GenerateIntFromCounter(SimpleCounter:NiagaraDataInterfaceSimpleCounter=DataInterface() {propertyValues {GpuSyncMode=None, InitialValue=0}})
         B=4
        )
        B=-2
       )))
    )
   ) {isSetParameters=true}
  ]
  particleUpdateScript [
   &ParticleState()
   &DynamicMaterialParameters(
    Index 0 Param 3=&FloatFromCurve(FloatCurve=DataInterface() {propertyValues {Curve {keys [
         <at 0 1 RCIM_Cubic>
         <at 0.2 0.025 RCIM_Cubic>
         {arriveTangent=2.9631, interpMode=RCIM_Cubic, leaveTangent=2.9631, tangentMode=RCTM_User, time=1, value=1}
        ]}}})
    Module.Param0WriteEnabled=false
    Module.Param1WriteEnabled=false
    Module.Param2WriteEnabled=true
    Module.Param3WriteEnabled=false
   )
   &ScaleColor(
    Scale RGB=&Multiply_VectorByFloat(
     Vector=&MakeVectorFromLinearColorRGB(Color=&ColorFromCurve(ColorCurve=DataInterface() {propertyValues {
         AlphaCurve {keys []}
         BlueCurve {keys [<at 0 1>, {arriveTangent=1, interpMode=RCIM_Cubic, leaveTangent=1, time=0.15, value=10}, <at 1 2 RCIM_Cubic>]}
         GreenCurve {keys [<at 0 1>, {arriveTangent=0.188, interpMode=RCIM_Cubic, leaveTangent=0.188, time=0.15, value=9.045}, <at 1 1.188 RCIM_Cubic>]}
         RedCurve {keys [<at 0 1>, {arriveTangent=0.1927, interpMode=RCIM_Cubic, leaveTangent=0.1927, time=0.15, value=8.5417}, <at 1 1.1927 RCIM_Cubic>]}
        }}))
     Float=1
    )
    ScaleA=false
    ScaleRGB=true
   )
   &ScaleSpriteSize(Scale Factor=&Vector2DFromFloat(Value=&LerpMultipleFloats(
      Float Count:ENiagara_MultipleLerpCount=::NewEnumerator0
      1:NiagaraFloat=1
      2:NiagaraFloat=0.25
      3:NiagaraFloat=0
      4:NiagaraFloat=?
      5:NiagaraFloat=?
      Alpha:NiagaraFloat=$Particles.NormalizedAge
     )))
   /Niagara/Modules/Update/SubUV/SubUVAnimation(SubUV Lookup Index=&Add_Float(A=&Multiply_Float(A=2, B=$Particles.NormalizedAge), B=$Particles.Initial.SubImageIndex))
   /SetVariables(Particles.SubImageIndexAlt:NiagaraFloat=&ModuloFloat(
     A=&Add_Float(A=$Particles.Initial.SubImageIndexAlt, B=&Multiply_Float(A=4, B=$Particles.NormalizedAge))
     B=4
    )) {isSetParameters=true}
   /SetVariables(
    Particles.SpriteFacing=&Add_Vector(
     A=$Particles.SpriteFacing
     B=&Multiply_VectorByFloat(Vector=$Particles.FacingAdjustment, Float=$Particles.NormalizedAge)
    )
    Particles.SpriteFacingAlt:Vector3f=&Add_Vector(
     A=$Particles.SpriteFacingAlt
     B=&Multiply_VectorByFloat(Vector=$Particles.FacingAdjustment, Float=$Particles.NormalizedAge)
    )
   ) {isSetParameters=true}
   /SetVariables(Particles.LightRadius=&Multiply_Float(
     A=$Particles.Initial.LightRadius
     B=&LerpMultipleFloats(
      Float Count:ENiagara_MultipleLerpCount=::NewEnumerator1
      1:NiagaraFloat=0
      2:NiagaraFloat=1
      3:NiagaraFloat=1
      4:NiagaraFloat=0
      5:NiagaraFloat=?
      Alpha:NiagaraFloat=$Particles.NormalizedAge
     )
    )) {isSetParameters=true}
   /SetVariables(
    Particles.LightColor=&MultiplyLinearColor(
     Linear Color=$Particles.Initial.Color
     Scale Factor=&MakeVector4_FromFloat(Float=$Particles.LightBrightness)
    )
    Particles.LightSourceRadius:NiagaraFloat=&Multiply_Float(
     A=$Particles.LightRadius
     B=&ClampFloat(Float=&Lerp_Float(A=2, B=-1, Alpha=$Particles.NormalizedAge))
    )
    Particles.LightTargetRadius:NiagaraFloat=&Multiply_Float(
     A=$Particles.LightRadius
     B=&ClampFloat(Float=&Lerp_Float(A=0, B=2, Alpha=$Particles.NormalizedAge))
    )
   ) {isSetParameters=true}
   /SetVariables(
    Particles.SourceVisibilityTag:NiagaraInt32=&self:Part_Hidden_SocketSelection(
     is Hand Hidden:NiagaraBool=$User.bisHandHidden
     is Head Hidden:NiagaraBool=$User.bisHeadHidden
     is Foot Hidden:NiagaraBool=$User.bisFootHidden
     Selection:NiagaraInt32=&MakeIntFromFloat(Float=$Particles.SourceBone)
    )
    Particles.TargetVisibilityTag:NiagaraInt32=&self:Part_Hidden_SocketSelection(
     is Hand Hidden:NiagaraBool=$User.bisHandHidden
     is Head Hidden:NiagaraBool=$User.bisHeadHidden
     is Foot Hidden:NiagaraBool=$User.bisFootHidden
     Selection:NiagaraInt32=&MakeIntFromFloat(Float=$Particles.TargetBone)
    )
   ) {isSetParameters=true}
  ]
  rendererClasses [&NiagaraSpriteRendererProperties, &NiagaraLightRendererProperties]
  renderers [
   {rendererClass=&NiagaraSpriteRendererProperties}
   {rendererClass=&NiagaraSpriteRendererProperties, rendererIndex=1}
   {rendererClass=&NiagaraLightRendererProperties, rendererIndex=2}
   {rendererClass=&NiagaraLightRendererProperties, rendererIndex=3}
  ]
 }
 Runes {
  emitterUpdateScript [
   &EmitterState()
   &SpawnRate(
    Spawn Probability=0.2
    SpawnRate=&UniformRangedFloat(
     Minimum=0
     Maximum=&ClampFloat(
      Float=&Multiply_Float(A=&MakeFloatFromInt(Int=$User.KillCount), B=4)
      Max=12
     )
    )
    Use Spawn Probability=true
   )
  ]
  particleSpawnScript [
   &InitializeParticle(
    Color=&MakeLinearColor(
     R:NiagaraFloat=1
     G:NiagaraFloat=&UniformRangedFloat(Minimum=1, Maximum=5)
     B:NiagaraFloat=&UniformRangedFloat(Minimum=1, Maximum=10)
     A:NiagaraFloat=&UniformRangedFloat(Minimum=0.5, Maximum=1)
    )
    Lifetime=&UniformRangedFloat(Minimum=0.1, Maximum=&UniformRangedFloat(Minimum=0.2, Maximum=1))
    Module.Write Color=true
    Module.Write Lifetime=true
    Module.Write Mass=false
    Module.Write Material Random=true
    Module.Write Position=false
    Module.Write Scale=false
    Module.Write SpriteRotation=true
    Module.Write SpriteSize=true
    Position=&SimulationPosition()
    Sprite Rotation=&UniformRangedFloat(Minimum=0, Maximum=360)
    Sprite Size=&Vector2DFromFloat(Value=&UniformRangedFloat(Minimum=2, Maximum=4))
   )
   /SetVariables(Particles.SpriteInitRotationRate:NiagaraFloat=&Multiply_Float(
     A=&UniformAOrBFloat(A=-1)
     B=&UniformRangedFloat(Minimum=5, Maximum=10)
    )) {isSetParameters=true}
   &AddVelocityFromPoint(Module.Use Falloff Distance=false, Velocity Strength=&UniformRangedFloat(Minimum=100, Maximum=200))
   &SubUVAnimation(SubUV Animation Mode=::NewEnumerator1)
   &SkeletalMeshLocation(
    Inherit Velocity=false
    MaskSpawnGroup=false
    Mesh Sampling Type=::NewEnumerator5
    Module.Override Randomness=false
    Sampled Position Offset=&Lerp_Vector(
     A=<10 0 0>
     B=<0 8 0>
     Alpha=&ClampFloat(Float=&Add_Float(A=&MakeFloatFromInt(Int=$Particles.PickBone), B=-2))
    )
    Sampled Position Offset Coordinate Space:ENiagaraMeshTransforms=::NewEnumerator3
    Skeletal Mesh=$User.SkeletalMesh
   ) {module=SkeletalMeshLocation001}
   /SetVariables(Particles.Socket Selection:NiagaraInt32=$Output.SkeletalMeshLocation001.SampledBoneIndex) {isSetParameters=true}
  ]
  particleUpdateScript [
   &ParticleState()
   &SpriteRotationRate(Rotation Rate=&Multiply_Float(A=$Particles.SpriteInitRotationRate, B=&VectorLength(Vector=$Particles.Velocity)))
   /SetVariables(Particles.Velocity=&Multiply_VectorByFloat(
     Vector=$Particles.Initial.Velocity
     Float=&ClampFloat(Float=&Lerp_Float(A=1, B=-1, Alpha=$Particles.NormalizedAge))
    )) {isSetParameters=true}
   &SolveForcesAndVelocity(Clamp Velocity=false, Limit Acceleration=false)
   &ScaleColor(
    Scale Alpha=&FloatFromCurve(FloatCurve=DataInterface() {propertyValues {Curve {keys [{interpMode=RCIM_Cubic, tangentMode=RCTM_User, time=0.75, value=1}, <at 1 0.5>]}}})
    Scale RGB=&MakeVector(
     X=&UniformAOrBFloat(A=0)
     Y=1
     Z=&FloatFromCurve(FloatCurve=DataInterface() {propertyValues {Curve {keys [
          {arriveTangent=-172.0061, interpMode=RCIM_Cubic, leaveTangent=-172.0061, tangentMode=RCTM_User, time=0.2, value=10}
          {arriveTangent=-0.0507, interpMode=RCIM_Cubic, leaveTangent=-0.0508, tangentMode=RCTM_User, time=0.32, value=1}
          {arriveTangent=90.7103, interpMode=RCIM_Cubic, leaveTangent=90.7105, tangentMode=RCTM_User, time=1, value=20}
         ]}}})
    )
    ScaleA=true
    ScaleRGB=true
   ) {module=ScaleColor001}
   &ScaleSpriteSize(Scale Factor=&Vector2DFromFloat(Value=&FloatFromCurve(FloatCurve=DataInterface() {propertyValues {Curve {keys [<at 0 1 RCIM_Cubic>, <at 0.1 1.5 RCIM_Cubic>, {interpMode=RCIM_Cubic, leaveTangent=0.8761, time=0.5, value=1}]}}})))
   /SetVariables(Particles.VisibilityTag=&self:Part_Hidden_SocketSelection(
     is Hand Hidden:NiagaraBool=$User.bisHandHidden
     is Head Hidden:NiagaraBool=$User.bisHeadHidden
     is Foot Hidden:NiagaraBool=$User.bisFootHidden
     Selection:NiagaraInt32=$Particles.Socket Selection
    )) {isSetParameters=true, module=SetVariables_#1}
  ]
  rendererClasses [&NiagaraSpriteRendererProperties]
  renderers [{rendererClass=&NiagaraSpriteRendererProperties}]
 }
}
summary {
 systemName=NS_Skin_CloakedAssassin
 emitters [
  {
   emitterName=LightningSprites
   rendererClasses [&NiagaraSpriteRendererProperties, &NiagaraLightRendererProperties]
  }
  {
   emitterName=BigLightningRibbon
   rendererClasses [&NiagaraRibbonRendererProperties, &NiagaraSpriteRendererProperties]
  }
  {emitterName=BigLightningEndPicker, rendererClasses [&NiagaraSpriteRendererProperties]}
  {emitterName=GlowMesh, rendererClasses [&NiagaraMeshRendererProperties, &NiagaraLightRendererProperties]}
  {emitterName=Runes, rendererClasses [&NiagaraSpriteRendererProperties]}
 ]
 userVariables [
  User.bisFootHidden {type=NiagaraBool}
  User.bIsFrontEndPreview {type=NiagaraBool}
  User.bisHandHidden {type=NiagaraBool}
  User.bisHeadHidden {type=NiagaraBool}
  User.KillCount {type=NiagaraInt32}
  User.SkeletalMesh {
   type=NiagaraDataInterfaceSkeletalMesh
   defaultValue {
    struct=NiagaraExt_VariableValue_DataInterface
    value {dataInterface=&self:NiagaraDataInterfaceSkeletalMesh_1}
   }
  }
 ]
}
